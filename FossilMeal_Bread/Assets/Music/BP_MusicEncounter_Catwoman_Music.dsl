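; Music encounter Blueprint for the Catwoman boss that adjusts boss music when the boss's conversion state changes.
; Conversion uninitializes the encounter; reverting initializes it and enters idle only if neither of two checked encounter states is set.
; Entering idle linearly fades the NPC music audio component to silence over the configured fade-out duration, when that component is valid.
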(fn BossConversionStateChanged
  (params (BossComponentInterface :type Music%20Encounter%20Component%20Boss%20Interface%20Interface) (Instigator :type Fort%20Pawn%20Object%20Reference) (Converted%20Pawn :type Fort%20Pawn%20Object%20Reference) (Converted :type Boolean))
  (if Converted
    (CallFunction|UninitializeMusicEncounter)
    (elif (Math|Boolean|NORBoolean (MusicEncounter|GetMusicEncounterStateValue "NewEnumerator2") (MusicEncounter|GetMusicEncounterStateValue "NewEnumerator1"))
      (CallFunction|InitializeMusicEncounter)
      (CallFunction|EnterIdleState))))

(fn EnterIdleState
  (bind _npc_music_audio_component (Variables|MusicEncounter|Boss|Music|GetNPC%20Music%20Audio%20Component)
    (:then
      (Audio|Components|Audio|FadeOut _npc_music_audio_component (Variables|MusicEncounter|Boss|Music|GetNPC%20Music%20Fade%20Out%20Duration) 0.0 "Linear" :!class /Script/Engine.AudioComponent))
    (:else)))
