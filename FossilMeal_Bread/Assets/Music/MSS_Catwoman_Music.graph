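# MetaSound Source preset of the default NPC music encounter, supplying the Catwoman encounter's music as spatialized stereo through NPC music attenuation and sound class.
# The parent exposes Idle, Combat and Defeated sections, with inputs for phase, tempo, pausing, fades and beat offsets.
# This preset fills Combat with separate percussion and non-percussion loop stems and Defeated with a death stinger, at 160 tempo.
# Idle and transition waves are left empty.
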MetaSoundSource /FossilMeal_Bread/Assets/Music/MSS_Catwoman_Music
preset_parent=/MusicEncounter/MetaSounds/MSS_MusicEncounter_NPC_Default
inputs [
 Combat.BeatOffset {type=Int32}
 Combat.ElseWave {type=WaveAsset}
 Combat.FadeInTime {type=Time}
 Combat.FadeOutTime {type=Time}
 Combat.PercWave {type=WaveAsset}
 Combat.TransitionVolume {type=Float}
 Combat.TransitionWave {type=WaveAsset}
 Combat.Volume {type=Float}
 Defeated.BeatOffset {type=Int32}
 Defeated.Volume {type=Float}
 Defeated.Wave {type=WaveAsset}
 Idle.BeatOffset {type=Int32}
 Idle.FadeInTime {type=Time}
 Idle.FadeOutTime {type=Time}
 Idle.TransitionVolume {type=Float}
 Idle.TransitionWave {type=WaveAsset}
 Idle.Volume {type=Float}
 Idle.Wave {type=WaveAsset}
 MusicEncounter.bIsPaused {type=Bool}
 MusicEncounter.OutputVolume {type=Float}
 MusicEncounter.PauseFadeOutTime {type=Time}
 MusicEncounter.Phase {type=Int32}
 MusicEncounter.Tempo {type=Float}
 UE.Source.OnPlay {type=Trigger}
]
interfaces [UE.OutputFormat.Stereo:1.0, UE.Source.OneShot:1.0, UE.Source:1.1]
outputs [
 "UE.OutputFormat.Stereo.Audio:0" {type=Audio}
 "UE.OutputFormat.Stereo.Audio:1" {type=Audio}
 UE.Source.OneShot.OnFinished {type=Trigger}
]
pages [Default {input_defaults {
   Combat.BeatOffset=1
   Combat.ElseWave=&/MX_S41_CatWoman_Phase_1_LP_Else
   Combat.FadeInTime=0
   Combat.FadeOutTime=1
   Combat.PercWave=&/MX_S41_CatWoman_Phase_1_LP_Perc
   Combat.TransitionVolume=0
   Combat.TransitionWave=null
   Combat.Volume=0
   Defeated.BeatOffset=2
   Defeated.Volume=0
   Defeated.Wave=&/MX_S41_CatWoman_Death_STNG
   Idle.BeatOffset=1
   Idle.FadeInTime=1
   Idle.FadeOutTime=1
   Idle.TransitionVolume=0
   Idle.TransitionWave=null
   Idle.Volume=0
   Idle.Wave=null
   MusicEncounter.OutputVolume=0
   MusicEncounter.PauseFadeOutTime=3
   MusicEncounter.Phase=1
   MusicEncounter.Tempo=160
   MusicEncounter.bIsPaused=null
   UE.Source.OnPlay=null
  }}]
properties {
 assetImportData=&self:AssetImportData
 attenuationSettings=/Game/Sounds/Attenuation/ATT_Music_NPC
 maxDistance=6500
 numChannels=2
 outputFormat=Stereo
 sampleRate=48000
 soundClassObject=&Fort_Music_NPC
 volume=1.6
}
